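# Conversation database for the NPC DJ Yonder, defining the greeting shown when a player talks to this character.
# The entry point leads to one speech node with a default line and contextual variants gated on first-ever conversation, conversation count, or the player wearing specific character cosmetics.
# That node branches through link nodes to shared NPC quests and basic services, DJ Yonder's dialogue, and Season 36 character stories for DJ Yonder and Big Dill.
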ConversationDatabase /NPCLibrary/NPCs/DJYonder/NPCLibrary_Conversation_DJYonder
entry_points [{tag=Conversation.Entry.DJYonder, destinations [#1]}]
nodes [
 ConversationEntryPointNode_7 {
  id=#1
  class=ConversationEntryPointNode
  outputs [#2]
  properties {entryTag {tagName=Conversation.Entry.DJYonder}}
 }
 Speech_C_4 {
  id=#2
  class=Speech_C
  outputs [
   #3
   #4
   #5
   #6
   #7
  ]
  properties {
   nodeName=""
   generalConfig {
    defaultMessage=If you've heard a remix on this Island: that's mine. Or perhaps Bop's.
    contextualMessages [
     {
      message="You already know my name: DJ Yonder!"
      contextRequirements [{requirement {class=FortControllerRequirement_IsFirstEverConversationWithNPC}}]
      priority {curve {rowName=Tandem.Speech.Priority.FirstEverInteraction}}
     }
     {
      message="I'm an artist. You wouldn't bother a painter for a selfie, would you?"
      contextRequirements [{requirement {class=FortControllerRequirement_CheckNPCMatchConversationCount, properties {nPCUniqueIdTag {tagName=AISpawnerData.Type.Tandem.DJYonder}}}}]
      priority {curve {rowName=None}}
     }
     {
      message=If you've heard a remix on this Island: that's mine. Or perhaps Bop's.
      contextRequirements [{requirement {class=FortControllerRequirement_CheckNPCMatchConversationCount, properties {nPCUniqueIdTag {tagName=AISpawnerData.Type.Tandem.DJYonder}}}}]
      priority {curve {rowName=None}}
     }
     {
      message="If you want to start a new 'beef', why don't you go bother your sausage friend?"
      contextRequirements [{requirement {class=FortControllerRequirement_HasCID, properties {allowedCIds [/BRCosmetics/Athena/Items/Cosmetics/Characters/Character_WartyBrine_Sus]}}}]
      priority {curve {rowName=None}}
     }
     {
      message=Control your pickle. He's been bothering me.
      contextRequirements [{requirement {
         class=FortControllerRequirement_HasCID
         properties {allowedCIds [/BRCosmetics/Athena/Items/Cosmetics/Characters/CID_614_Athena_Commando_M_MissingLink]}
        }}]
      priority {curve {rowName=None}}
     }
     {
      message=These Tasty Bois are getting in our way. Let's feed them to the wolves.
      contextRequirements [{requirement {
         class=FortControllerRequirement_HasCID
         properties {allowedCIds [/BRCosmetics/Athena/Items/Cosmetics/Characters/CID_308_Athena_Commando_F_FortniteDJ]}
        }}]
      priority {curve {rowName=None}}
     }
    ]
   }
  }
 }
 ConversationLinkNode_4 {
  id=#7
  class=ConversationLinkNode
  properties {remoteEntryTag {tagName=Conversation.Entry.Season36.DJYonder.CharacterStories.01}}
 }
 ConversationLinkNode_3 {
  id=#6
  class=ConversationLinkNode
  properties {remoteEntryTag {tagName=Conversation.Entry.Season36.BigDill.CharacterStories.01}}
 }
 ConversationLinkNode_2 {id=#5, class=ConversationLinkNode, properties {remoteEntryTag {tagName=Conversation.Entry.DJYonder.Dialogue}}}
 ConversationLinkNode_0 {id=#4, class=ConversationLinkNode, properties {remoteEntryTag {tagName=Conversation.Entry.BasicServices}}}
 ConversationLinkNode_1 {id=#3, class=ConversationLinkNode, properties {remoteEntryTag {tagName=Conversation.Entry.NPCLibrary.Quests}}}
]
